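# Masked master material for the Candygram vehicle wheel, built on the MF_WheelParent_Simple function for skeletal meshes.
# It adds a fourth rim color, metallic and roughness from the rim mask's alpha, limited by inverted red vertex color.
# Blue vertex color blends the opacity mask toward the LOD rim texture's alpha, scaled by that inverted red.
# The result feeds the shared vehicle damage function, then the de-rez transition.
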Material /VehicleCosmetics/Wheels/Candygram/Materials/M_Wheel_Candygram
graph {
 properties {
  bAutomaticallySetUsageInEditor=false
  bUseMaterialAttributes=true
  bUsedWithMeshDeformer=true
  bUsedWithSkeletalMesh=true
  blendMode=BLEND_Masked
  pixelDepthOffsetMode=PDOM_Legacy
  usedShadingModels=Default Lit
 }
 nodes [
  n1 = MaterialFunctionCall() {materialFunction=&MF_WheelParent_Simple}
  n2 = GetMaterialAttributes(None=@n1.MaterialAttributes) {attributeGetTypes [
    #1
    #2
    #3
    #4
   ]}
  n3 = ComponentMask(None=@n1.RimRGBAMask) {a=true}
  n4 = Multiply(A=VectorParameter() {parameterName=Rim_ColorD, defaultValue=<rgba 0.013 0.0666 0.5841 1>}.RGB, B=@n3)
  n5 = OneMinus(None=@n1.VertexColorR)
  n6 = VectorParameter() {parameterName=RimColorD_PBR, channelNames {a=Unuse, b=Metallic, g=Roughness, r=Unuse}, defaultValue=<rgba 0 0.5029 1 1>}
  n7 = Multiply(A=@n6.Metallic, B=@n3)
  n8 = Multiply(A=@n6.Roughness, B=@n3)
  n9 = TextureSample() {texture=/VehicleCosmetics/Wheels/Candygram/Textures/LOD/T_Rim_Candygram_BC_LOD}
  n10 = SetMaterialAttributes!(
   MaterialAttributes=@n2.MaterialAttributes
   Base Color=LinearInterpolate(A=@n2.Base Color, B=Add(A=@n4, B=@n2.Base Color), Alpha=@n5)
   Metallic=LinearInterpolate(A=@n2.Metallic, B=Add(A=@n7, B=@n2.Metallic), Alpha=@n5)
   Roughness=LinearInterpolate(A=@n2.Roughness, B=Add(A=@n8, B=@n2.Roughness), Alpha=@n5)
   Opacity Mask=LinearInterpolate(A=@n2.Opacity Mask, B=Multiply(A=@n9.A, B=@n5), Alpha=VertexColor().B)
  ) {attributeSetTypes [
    #1
    #2
    #3
    #4
   ]}
  MaterialFunctionCall(In=MaterialFunctionCall(Material=@n10) {materialFunction=&MF_BaseVehicle_Damage}.Result) {materialFunction=&MF_VehicleDeRez_Transition}
 ]
}
